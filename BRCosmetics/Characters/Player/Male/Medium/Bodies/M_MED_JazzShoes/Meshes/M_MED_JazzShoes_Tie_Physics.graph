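# Physics asset for the JazzShoes male medium character body, simulating secondary motion on its necktie and apron tie.
# Two three-bone capsule chains hang from spine_05 (tie) and pelvis (apron tie), with twist limited to zero degrees.
# The tie has tighter swing limits than the apron tie.
# Larger capsules on pelvis, upper spine and left thigh, joined by free-motion constraints, form body shapes around the chains, likely for collision.
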PhysicsAsset /BRCosmetics/Characters/Player/Male/Medium/Bodies/M_MED_JazzShoes/Meshes/M_MED_JazzShoes_Tie_Physics
bodies {
 dyn_apronTie_01 [{
   shapeType=Capsule
   length=2.1265
   radius=0.8726
   shapeName=dyn_apronTie_01_capsule
   center=<2.1906 -0.1229 -0.0203>
   rotation=<rot 0 -88.5033 -82.5308>
  }]
 dyn_apronTie_02 [{
   shapeType=Capsule
   length=4.1724
   radius=1.5251
   shapeName=dyn_apronTie_02_capsule
   center=<4.2612 0.0005 0.2708>
   rotation=<rot 0 94.8346 -89.6134>
  }]
 dyn_apronTie_03 [{
   shapeType=Capsule
   length=5.1541
   radius=2.0956
   shapeName=dyn_apronTie_03_capsule
   center=<5.9026 0.2342 0.2617>
   rotation=<rot 0 90.5903 -88.0154>
  }]
 dyn_tie_01 [{shapeType=Capsule, length=3.309, radius=1.6465, shapeName=dyn_tie_01_capsule, center=<2.3057 0 0>, rotation=<rot 90 0 0>}]
 dyn_tie_02 [{shapeType=Capsule, length=3.1403, radius=1.3063, shapeName=dyn_tie_02_capsule, center=<2.3057 0 0>, rotation=<rot 90 0 0>}]
 dyn_tie_03 [{shapeType=Capsule, length=5.1083, radius=1.0651, shapeName=dyn_tie_03_capsule, center=<3.7511 0 0>, rotation=<rot 90 0 0>}]
 pelvis [{shapeType=Capsule, length=10, radius=13, shapeName=pelvis_capsule, center=<-8.7498 -2.148 0>}]
 spine_04 [{shapeType=Capsule, length=12, radius=15, shapeName=spine_04_capsule, center=<-3.2667 2.5258 0>}]
 spine_05 [{shapeType=Capsule, length=14.5, radius=12.5, shapeName=spine_05_capsule, center=<-1.0702 0.4253 0>}]
 thigh_l [{shapeType=Capsule, length=31.5, radius=10, shapeName=thigh_l_capsule, center=<-20.5916 0 0.6967>, rotation=<rot 90 90 90.0001>}]
}
constraints [
 {
  bone1Name=spine_05
  bone2Name=spine_04
  swing1LimitDegrees=45
  swing1Motion=Free
  swing2LimitDegrees=45
  swing2Motion=Free
  twistLimitDegrees=45
  twistMotion=Free
 }
 {bone1Name=dyn_tie_03, bone2Name=dyn_tie_02, swing1LimitDegrees=10, swing2LimitDegrees=20, twistLimitDegrees=0}
 {bone1Name=dyn_tie_01, bone2Name=spine_05, swing1LimitDegrees=10, swing2LimitDegrees=20, twistLimitDegrees=0}
 {bone1Name=dyn_tie_02, bone2Name=dyn_tie_01, swing1LimitDegrees=10, swing2LimitDegrees=20, twistLimitDegrees=0}
 {bone1Name=dyn_apronTie_03, bone2Name=dyn_apronTie_02, swing1LimitDegrees=45, swing2LimitDegrees=45, twistLimitDegrees=0}
 {bone1Name=dyn_apronTie_02, bone2Name=dyn_apronTie_01, swing1LimitDegrees=45, swing2LimitDegrees=45, twistLimitDegrees=0}
 {
  bone1Name=spine_04
  bone2Name=pelvis
  swing1LimitDegrees=45
  swing1Motion=Free
  swing2LimitDegrees=45
  swing2Motion=Free
  twistLimitDegrees=45
  twistMotion=Free
 }
 {
  bone1Name=thigh_l
  bone2Name=pelvis
  swing1LimitDegrees=45
  swing1Motion=Free
  swing2LimitDegrees=45
  swing2Motion=Free
  twistLimitDegrees=45
  twistMotion=Free
 }
 {bone1Name=dyn_apronTie_01, bone2Name=pelvis, swing1LimitDegrees=45, swing2LimitDegrees=45, twistLimitDegrees=0}
]
